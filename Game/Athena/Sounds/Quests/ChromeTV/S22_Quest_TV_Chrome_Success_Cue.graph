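# Sound cue for the Chrome TV quest success sound, played on the UI sound class with its own attenuation override.
# A mixer layers the success wave over quieter Chrome grenade glass and goop explosion waves.
# The root quality-level node plays that mix for the first two levels and only the success wave for the last two.
# Concurrency is overridden to a single instance that stops the oldest.
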SoundCue /Game/Athena/Sounds/Quests/ChromeTV/S22_Quest_TV_Chrome_Success_Cue
root=QualityLevel_0
nodes [
 {id=Mixer_0, children [WavePlayer_3,WavePlayer_4,WavePlayer_0], properties {inputVolume [0.6,0.6,1]}}
 {id=QualityLevel_0, children [Mixer_0,Mixer_0,WavePlayer_0,WavePlayer_0]}
 {id=WavePlayer_0, properties {soundWaveAssetPtr=&/Chrome_TV_Success}}
 {id=WavePlayer_3, properties {soundWaveAssetPtr=&/Chrome_Grenade_Explo_Close_Glass_v2_04}}
 {id=WavePlayer_4, properties {soundWaveAssetPtr=&/Chrome_Grenade_Explo_Close_Goop_v2_02}}
]
properties {
 bOverrideAttenuation=true
 bOverrideConcurrency=true
 maxDistance=1250
 soundClassObject=&Fort_UI
 volumeMultiplier=0.7
 attenuationOverrides {falloffDistance=850, stereoSpread=2300}
 concurrencyOverrides {maxCount=1, resolutionRule=StopOldest}
}
