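# Looping ambient static sound cue for a chrome TV prop, filed under Athena quest sounds and routed to the tile-mix ambient class.
# The first two quality levels mix a static loop and a chrome tone loop, each through its own attenuation node.
# The other two play a single optimized static loop without an attenuation node.
# A concurrency override prevents new instances once the limit is reached.
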SoundCue /Game/Athena/Sounds/Quests/ChromeTV/TV_Chrome_Static_Loop_Cue
root=QualityLevel_0
nodes [
 {id=Attenuation_0, children [WavePlayer_0], properties {attenuationSettings=/Game/Sounds/Attenuation/Fort_Prop_TV_Athena_Att}}
 {id=Attenuation_1, children [WavePlayer_2], properties {attenuationSettings=&/TV_Chrome_Static_Loop_Att}}
 {id=Mixer_0, children [Attenuation_0, Attenuation_1], properties {inputVolume [0.75,1]}}
 {id=QualityLevel_0, children [Mixer_0,Mixer_0,WavePlayer_1,WavePlayer_1]}
 {id=WavePlayer_0, properties {bLooping=true, soundWaveAssetPtr=&/TV_Chrome_Static_Loop_01}}
 {id=WavePlayer_1, properties {bLooping=true, soundWaveAssetPtr=&/TV_Chome_Static_Loop_Opt_01}}
 {id=WavePlayer_2, properties {bLooping=true, soundWaveAssetPtr=&/Chrome_Tone_Loop_01}}
]
properties {
 bOverrideConcurrency=true
 maxDistance=1100
 soundClassObject=/Game/Sounds/FortSoundClassesAndModes/Fort_Ambient_TileMix
 volumeMultiplier=0.5
 attenuationOverrides {lPFRadiusMin=0}
 concurrencyOverrides {resolutionRule=PreventNew}
}
