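# Looping ambient cue for the static and tuning noise of a Chrome TV quest prop, spatialized with Athena TV attenuation.
# A Modulator wraps a loop over a quality switch that picks one of two mixes.
# Both layer a static-tuning bed with a quieter static loop faded in by an Enveloper.
# The first two quality slots also add randomly chosen tuning one-shots after a short random delay; the last two omit them.
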SoundCue /Game/Athena/Sounds/Quests/ChromeTV/TV_Chrome_Static_Tuning_Loop_Cue
root=Modulator_0
nodes [
 {id=Delay_1, children [Random_0], properties {delayMax=1, delayMin=0.5}}
 {
  id=Enveloper_0
  children [WavePlayer_10]
  properties {pitchCurve {editorCurveData {keys [{time=0}]}}, volumeCurve {editorCurveData {keys [<at 0 0>,<at 0.2 1>]}}}
 }
 {id=Looping_0, children [QualityLevel_0]}
 {id=Mixer_0, children [Delay_1,WavePlayer_7,Enveloper_0], properties {inputVolume [1,1,0.6]}}
 {id=Mixer_1, children [WavePlayer_7,Enveloper_0], properties {inputVolume [1,0.6]}}
 {id=Modulator_0, children [Looping_0]}
 {id=QualityLevel_0, children [Mixer_0,Mixer_0,Mixer_1,Mixer_1]}
 {id=Random_0, children [WavePlayer_2,WavePlayer_3,WavePlayer_4,WavePlayer_5,WavePlayer_6], properties {weights [1,1,1,1,1]}}
 {id=WavePlayer_10, properties {soundWaveAssetPtr=&/TV_Chrome_Static_Loop_01}}
 {id=WavePlayer_2, properties {soundWaveAssetPtr=&/TV_Chrome_Tuning_1}}
 {id=WavePlayer_3, properties {soundWaveAssetPtr=&/TV_Chrome_Tuning_2}}
 {id=WavePlayer_4, properties {soundWaveAssetPtr=&/TV_Chrome_Tuning_3}}
 {id=WavePlayer_5, properties {soundWaveAssetPtr=&/TV_Chrome_Tuning_4}}
 {id=WavePlayer_6, properties {soundWaveAssetPtr=&/TV_Chrome_Tuning_5}}
 {id=WavePlayer_7, properties {soundWaveAssetPtr=&/TV_Chrome_Static_Tuning_Loop}}
]
properties {
 attenuationSettings=/Game/Sounds/Attenuation/Fort_Prop_TV_Athena_Att
 bOverrideConcurrency=true
 maxDistance=500
 soundClassObject=/Game/Sounds/FortSoundClassesAndModes/Fort_Ambient_TileMix
 volumeMultiplier=0.55
 attenuationOverrides {lPFRadiusMin=0}
 concurrencyOverrides {resolutionRule=PreventNew}
}
